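# Idle Niagara effect for the KneeLens pickaxe, drawing a bluish glow attached to the weapon's skeletal mesh.
# A single GlowCard emitter burst-spawns persistent particles that follow a sampled bone or socket, with facing and offsets rotated to match.
# Time-of-day color and brightness feed mesh, airglow, outer exhaust and light attributes, drawn by two mesh renderers and one sprite renderer.
# User.bisSwinging drops the outer exhaust alpha to zero during swings.
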NiagaraSystem /BRCosmetics/Weapons/FORT_Melee/Pickaxe_KneeLens/FX/NS_Pickaxe_KneeLens_Idle
dependencies {
 usedDataInterfaces [&NiagaraDataInterfaceArrayFloat3]
 usedDynamicInputs []
 usedModules [&SystemState]
}
emitters {GlowCard {
  emitterUpdateScript [
   &EmitterState(
    Inactive Response=::NewEnumerator1
    Life Cycle Mode=::NewEnumerator1
    Loop Behavior=::NewEnumerator1
    Loop Duration Mode=::NewEnumerator1
    UseLoopDelay=false
   )
   &SpawnBurst_Instantaneous(Use Spawn Probability=false, UseLoopCountLimit=false)
  ]
  particleSpawnScript [
   &InitializeParticle(
    Color=&FNTimeOfDayColor(
     Night Color=<rgba 0.5104 1.1135 2 1>
     Morning Color=<rgba 0.9271 1.3615 2 1>
     Day Color=<rgba 0.9271 1.3615 2 1>
     Evening Color=<rgba 0.9271 1.3615 2 1>
    )
    Position Offset=&SelectVectorFromArray(
     Direct Array Index=&SpawnGroup()
     Vector Selection Array=DataInterface() {propertyValues {FloatData [<70 54 28>, <70 -54 28>, <76 67 16>, <76 -67 16>]}}
    )
    Sprite Size Mode=::NewEnumerator3
    UsePositionOffset=false
    Write Material Random=true
   )
   &CameraOffset(Camera Offset Amount=1)
  ]
  particleUpdateScript [
   &ParticleState(Kill Particles When Lifetime Has Elapsed=false)
   &SkeletalMeshLocation(
    Bone / Socket Index=2
    MaskSpawnGroup=false
    Mesh Sampling Type=::NewEnumerator5
    Sampled Position Offset=<0 -2.5 0>
    Skeletal Mesh=$User.SkeletalMesh
    Socket Sampling Mode=::NewEnumerator5
   )
   &SpriteFacingAndAlignment(
    Sprite Facing=&RotateVectorByQuaternion(Vector To Rotate=<0 -1 0>, Quaternion=$Particles.SkeletalMeshLocation.SampledRotation)
    WriteAlignment=false
    WriteFacing=true
   )
   &FN_Light_Attributes(
    Color=$Particles.Color
    Create Color Attribute=true
    Create Exponent Attribute=true
    Create Position Attribute=true
    Create Radius Attribute=true
    Exponent=1
    Position=&AddVectorToPosition(Position=$Particles.Position, Vector=<0 -10 0>)
    Radius=50
    Scale Color=10
   )
   /SetVariables(
    Particles.MeshColor:LinearColor=&Multiply_LinearColorByFloat(
     Linear Color=&FNTimeOfDayColor(
      Night Color=<rgba 0.5104 1.1135 2 1>
      Morning Color=<rgba 0.9271 1.3615 2 1>
      Day Color=<rgba 0.9271 1.3615 2 1>
      Evening Color=<rgba 0.9271 1.3615 2 1>
     )
     Float=&Multiply_Float(A=1.8, B=&FN_Select_Float_from_Time_Of_Day(Morning=1, Day=1, Evening=1, Night=0.8))
    )
    Particles.MeshPosition:NiagaraPosition=&AddVectorToPosition(
     Position=$Particles.Position
     Vector=&RotateVectorByQuaternion(Vector To Rotate=<0 2 0>, Quaternion=$Particles.SkeletalMeshLocation.SampledRotation)
    )
   ) {isSetParameters=true}
   &ScaleColor(
    Scale Alpha=&Multiply_Float(A=1, B=&FN_Select_Float_from_Time_Of_Day(Morning=1, Day=1, Evening=1, Night=0.8))
    ScaleA=true
    ScaleRGB=true
   )
   /SetVariables(
    Particles.AirglowColor:LinearColor=&Multiply_LinearColorByFloat(Linear Color=$Particles.Color, Float=0.4)
    Particles.AirglowPosition:NiagaraPosition=&AddVectorToPosition(
     Position=$Particles.Position
     Vector=&RotateVectorByQuaternion(Vector To Rotate=<0 -3 0>, Quaternion=$Particles.SkeletalMeshLocation.SampledRotation)
    )
    Particles.AirglowSize:Vector2f=&Multiply_Vector2D_ByFloat(Float=1.8, Vector2D=$Particles.SpriteSize)
   ) {isSetParameters=true}
   /SetVariables(
    Particles.OuterExhaustColor:LinearColor=&MultiplyLinearColor(
     Linear Color=$Particles.Color
     Scale Factor=&MakeFloat4FromVectorAndFloat(
      "Vector (RGB)"=<1 1 1>
      "Float (Alpha)"=&Lerp_Float(A=0.08, B=0, Alpha=&MakeCustomFloatFromBool(Bool=$User.bisSwinging, True Float=1, False Float=0))
     )
    )
    Particles.OuterExhaustPosition:NiagaraPosition=&AddVectorToPosition(
     Position=$Particles.SkeletalMeshLocation.SampledPosition
     Vector=&RotateVectorByQuaternion(Vector To Rotate=<0 -5 0>, Quaternion=$Particles.SkeletalMeshLocation.SampledRotation)
    )
    Particles.OuterExhaustScale:Vector3f=&Multiply_VectorByFloat(Vector=<1 1 1.2>, Float=0.06)
   ) {isSetParameters=true}
  ]
  rendererClasses [&NiagaraMeshRendererProperties, &NiagaraSpriteRendererProperties]
  renderers [
   {rendererClass=&NiagaraMeshRendererProperties, rendererIndex=0}
   {rendererClass=&NiagaraSpriteRendererProperties, rendererIndex=1}
   {rendererClass=&NiagaraMeshRendererProperties, rendererIndex=2}
  ]
 }}
summary {
 systemName=NS_Pickaxe_KneeLens_Idle
 emitters [{emitterName=GlowCard, rendererClasses [&NiagaraMeshRendererProperties, &NiagaraSpriteRendererProperties]}]
 userVariables [
  User.bIsFrontEnd {type=NiagaraBool}
  User.bIsFrontEndPreview {type=NiagaraBool}
  User.bisSwinging {type=NiagaraBool}
  User.RibbonIDOffset {type=NiagaraInt32}
  User.SkeletalMesh {
   type=NiagaraDataInterfaceSkeletalMesh
   defaultValue {
    struct=NiagaraExt_VariableValue_DataInterface
    value {dataInterface=&self:NiagaraDataInterfaceSkeletalMesh_0}
   }
  }
 ]
}
