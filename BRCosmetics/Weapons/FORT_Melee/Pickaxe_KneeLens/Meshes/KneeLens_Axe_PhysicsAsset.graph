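# Physics asset for the KneeLens pickaxe axe mesh, a two-handed melee weapon cosmetic.
# It holds a single capsule body on the Melee_TwoHanded_Handle bone, offset and rotated to enclose the weapon.
# With no constraints, the whole axe is one rigid collision shape.
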PhysicsAsset /BRCosmetics/Weapons/FORT_Melee/Pickaxe_KneeLens/Meshes/KneeLens_Axe_PhysicsAsset
bodies {Melee_TwoHanded_Handle [{shapeType=Capsule, length=67.9654, radius=32.4518, center=<1.5554 2.5232 31.6555>, rotation=<rot 0.7438 -90 -0.5303>}]}
constraints []
